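; Test explosive actor that auto-detonates on the server about 12 seconds after BeginPlay, destroying its surroundings floor by floor upward.
; Detonate chains into a self-repeating NextFloorDestruction, stepping box overlaps up per floor with shrinking, clamped extents until max floors, then destroys itself.
; Overlapped actors of listed classes take lethal environment damage or are hidden if undamageable; debug boxes show each volume.
; ClientFX scales a glow via a once-only, remote-side timeline.
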(event EventBeginPlay
  (Utilities|FlowControl|Delay (Math|Random|RandomFloatinRange 12.0 13.0))
  (Networking|SwitchHasAuthority
    (:Authority
      (CallFunction|Detonate))))

(event Custom|Detonate
  (Networking|SwitchHasAuthority
    (:Authority
      _)
    (:Remote
      _))
  (CallFunction|ClientFX)
  (Utilities|FlowControl|Delay 0.1)
  (CallFunction|DestroyActorsinRadius (Variables|Default|GetCurrentMaxFloorsInt))
  (if (>= (Variables|Default|GetCurrentMaxFloorsInt) (Variables|Default|GetMaxFloorsInt))
    (Actor|DestroyActor)
    (else
      (Math|Integer|IncrementInt (Variables|Default|GetCurrentMaxFloorsInt)
        (:%20%20
          (CallFunction|NextFloorDestruction))))))

(event Custom|NextFloorDestruction
  (Networking|SwitchHasAuthority
    (:Authority
      _)
    (:Remote
      _))
  (CallFunction|ClientFX)
  (Utilities|FlowControl|Delay 0.1)
  (CallFunction|DestroyActorsinRadius (Variables|Default|GetCurrentMaxFloorsInt))
  (if (>= (Variables|Default|GetCurrentMaxFloorsInt) (Variables|Default|GetMaxFloorsInt))
    (Actor|DestroyActor)
    (else
      (Math|Integer|IncrementInt (Variables|Default|GetCurrentMaxFloorsInt)
        (:%20%20
          (CallFunction|NextFloorDestruction))))))

(event Custom|ClientFX
  (Utilities|FlowControl|DoOnce false
    (:Completed
      (Networking|SwitchHasAuthority
        (:Remote
          (bind (_direction _spherescale) (Timeline|Timeline PlayFromStart)
            (:Update
              (Transformation|SetWorldScale3D (Variables|Default|GetExplosionGlow) (Math|Vector|Lerp(Vector) "1.000000,1.000000,1.000000" "55.000000,55.000000,80.000000" _spherescale)))))))))

(fn DestroyActorsInRadius
  (params (FloorInt :type Integer))
  (LocalVariables|SetBoxPositionLocal (Math|Vector|MakeVector (.x (Transformation|GetWorldLocation (Variables|Default|GetCube))) (.y (Transformation|GetWorldLocation (Variables|Default|GetCube))) (+ (.z (Transformation|GetWorldLocation (Variables|Default|GetCube))) (* (Math|Conversions|ToFloat(Integer) FloorInt) 384.0))))
  (Variables|Default|SetCurrentDestructionExtent (Math|Vector|ClampVectorSize (- (Variables|Default|GetInitialDestructionExtent) (Math|Vector|MakeVector (* (LocalVariables|GetDecreasePerFloorLocal) (* FloorInt 1)) (* (LocalVariables|GetDecreasePerFloorLocal) (* FloorInt 1)))) 1.5e+03 2e+04))
  (for _array_element (LocalVariables|GetAllActorClasses)
    (bind _outactors (Collision|BoxOverlapActors (LocalVariables|GetBoxPositionLocal) (Variables|Default|GetCurrentDestructionExtent) (Utilities|Array|MakeArray "ObjectTypeQuery1") _array_element :!class /Script/Engine.KismetSystemLibrary))
    (Rendering|Debug|DrawDebugBox :Center (LocalVariables|GetBoxPositionLocal) :Extent (Variables|Default|GetCurrentDestructionExtent) :LineColor "(R=1.000000,G=0.000000,B=0.022420,A=1.000000)" :Duration 1.0 :Thickness 3.0 :DepthPriority "World")
    (for _array_element_2 _outactors
      (Utilities|IsValid _array_element_2
        (:Is%20Valid
          (if (Variables|/Script/Engine.Actor|GetbCanBeDamaged _array_element_2)
            (Environment|ApplyGlobalEnvironmentDamagetoActor _array_element_2 9999.0)
            (Rendering|SetActorHiddenInGame _array_element_2 true)))))))

(fn ConstructionScript)
