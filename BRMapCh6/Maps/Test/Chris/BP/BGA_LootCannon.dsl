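; Loot cannon that players light via a "Light Fuse" interaction, firing tiered loot pickups after a fuse delay.
; Interacting disables further interaction and calls FireLootCannon, which also accepts a loot tier as a custom event.
; On authority, it spawns pickups, tosses them along the arrow with random spread, and launches non-DBNO players overlapping the front box at half velocity.
; Non-authority machines play fuse and firing effects, audio and nearby-player camera shake.
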(event Interaction|EventBlueprintOnInteract
  (params (InteractingPawn :type Fort%20Pawn%20Object%20Reference) (InteractionBeingAttempted :type EInteractionBeingAttempted%20Enum :default "FirstInteraction"))
  ; PlayerInteract
  (Variables|Default|SetPlayerInteractable false)
  (Utilities|FlowControl|DoOnce false
    (:Completed
      (CallFunction|FireLootCannon (Variables|Default|GetLootTierToFire)))))

(event Custom|FireLootCannon
  (params (LootTierToFire :type Name))
  (bind _arrow (Variables|Default|GetArrow))
  (bind _returnvalue (Transformation|GetForwardVector _arrow))
  (bind _tossvelocity (Variables|Default|GetTossVelocity))
  (bind _returnvalue_2 (Math|Conversions|ToVector(Float) _tossvelocity))
  (bind _returnvalue_3 (* _returnvalue _returnvalue_2))
  ; Fire the Cannon
  (Variables|Default|SetLootTierToFire LootTierToFire)
  (if (Variables|Default|GetPlayerInteractable)
    (Utilities|FlowControl|DoOnce false
      (:Completed
        _)) ; only allow once shot if player interactable
    (else
      _))
  (CallFunction|LightFuse)
  (Utilities|FlowControl|Delay (Variables|Default|GetShootFuseDelay))
  (Networking|SwitchHasAuthority
    (:Authority
      (CallFunction|LaunchPlayers)
      (Variables|Default|SetImpulse _returnvalue_3)
      (bind _returnvalue_4 (Inventory|SpawnPickupInWorldWithLootTier (Variables|Default|GetLootTierToFire) (Transformation|GetWorldLocation (Variables|Default|GetArrow)) 0 0 true true _returnvalue_3))
      (Variables|Default|SetNewVar _returnvalue_4)
      (for _array_element (Variables|Default|GetNewVar)
        (bind _asfort_pickup_athena (Utilities|Casting|CastToFortPickupAthena _array_element))
        (Pickup|TossPickupinDirection _asfort_pickup_athena (+ (Math|Vector|MakeVector (* (Variables|Default|GetTossSpreadMultiplier) (Math|Random|RandomFloatinRange -4e+02 4e+02)) (* (Math|Random|RandomFloatinRange -4e+02 4e+02) (Variables|Default|GetTossSpreadMultiplier)) (* (Math|Random|RandomFloatinRange 1e+02 2.5e+02) (Variables|Default|GetTossSpreadMultiplier))) (* (Transformation|GetForwardVector (Variables|Default|GetArrow)) (Variables|Default|GetTossVelocity))) self false false)))
    (:Remote
      (CallFunction|CameraShake)
      (Components|Activation|Activate (Variables|Default|GetCannonFireEffect) false)
      (Audio|Components|Audio|Play (Variables|Default|GetLootLaunchAudio) 0.0)
      (Utilities|FlowControl|Delay 1.25)
      (Audio|Components|Audio|Play (Variables|Default|GetLootAudio) 0.0))))

(event Custom|LaunchPlayers
  ; Launch Player
  (for _array_element (Collision|GetOverlappingActors (Variables|Default|GetFrontBox) "/Script/FortniteGame.FortPlayerPawnAthena" :!class /Script/Engine.PrimitiveComponent)
    (bind _asfort_player_pawn_athena (Utilities|Casting|CastToFortPlayerPawnAthena _array_element))
    (if (not (Health|IsDBNO _asfort_player_pawn_athena))
      (Character|LaunchCharacter _asfort_player_pawn_athena (* (Transformation|GetForwardVector (Variables|Default|GetArrow)) (Math|Conversions|ToVector(Float) (Math|Float|SafeDivide (Variables|Default|GetTossVelocity) 2.0))) true true))))

(event Custom|CameraShake
  ; Camera Shake
  (Networking|SwitchHasAuthority
    (:Remote
      (bind _outactors (Collision|SphereOverlapActors (Transformation|GetWorldLocation (Variables|Default|GetStaticMesh)) 7.5e+02 (Utilities|Array|MakeArray "ObjectTypeQuery3") "/Script/FortniteGame.FortPlayerPawnAthena"))
      (for _array_element _outactors
        (bind _asfort_player_controller_athena (Utilities|Casting|CastToFortPlayerControllerAthena (Fortnite|GetControllerfromActor _array_element)))
        (Game|Feedback|ClientStartCameraShake :self _asfort_player_controller_athena :Shake "/Game/Effects/Fort_Effects/Camera_Effects/BP_CameraShake_Cannon_Fire.BP_CameraShake_Cannon_Fire_C" :Scale 1.0 :PlaySpace "World")))))

(event Custom|LightFuse
  ; Fuse Cosmetics
  (Networking|SwitchHasAuthority
    (:Remote
      (Components|Activation|Activate (Variables|Default|GetFuseEffect) false)
      (Audio|Components|Audio|Play (Variables|Default|GetFuseAudioCue) 0.0)
      (Utilities|FlowControl|Delay (Variables|Default|GetShootFuseDelay))
      (Components|Activation|Deactivate (Variables|Default|GetFuseEffect))
      (Audio|Components|Audio|FadeOut (Variables|Default|GetFuseAudioCue) 0.25 0.0 "Linear" :!class /Script/Engine.AudioComponent))))

(fn BlueprintCanInteract
  (params (InteractingPawn :type Fort%20Pawn%20Object%20Reference) (InteractionBeingAttempted :type EInteractionBeingAttempted%20Enum :default "FirstInteraction") (InteractionType :type TInteractionType%20Enum :default "IT_NoInteraction"))
  (return (Variables|Default|GetPlayerInteractable)))

(fn BlueprintGetInteractionString
  (params (InteractingPawn :type Fort%20Pawn%20Object%20Reference) (InteractionBeingAttempted :type EInteractionBeingAttempted%20Enum :default "FirstInteraction"))
  (return "NSLOCTEXT(\"[F06B915BD23D65EB5144084C805354E5]\", \"F7CE37CE486D5B4082CDA08206880C31\", \"Light Fuse\")"))

(fn ConstructionScript)
