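; BGA_MiniGameCK is a test tic-tac-toe-style minigame actor where a player plays an AI until one reaches two round wins.
; It spawns nine indexed button actors at marker transforms and matches owned indices against winning-combination structs.
; AI helpers trial shuffled free buttons for winning or blocking moves, plus a probability-based dumb check and random pick.
; Round results update button states, score, audio and emotion; OnRep functions refresh score text, face material and force field.
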(fn AIFindAnyWinningMove
  (LocalVariables|SetAllCachedOwnedIndexLocal (Variables|Default|GetAllAIOwnedButtonInts)) ; checks for a winning move for the ai
  ; Get All Non Pressed Button Indexs
  (for _array_element (Variables|Default|GetAllSpawnedButtonRefs)
    (switch Utilities|FlowControl|Switch|SwitchonEMiniGameCKButtonStates (Variables|/BRMapCh6/Maps/Test/Chris/BP/BGA_Minigame_Button_CK.BGA_Minigame_Button_CK_C|GetButtonState _array_element)
      (:NewEnumerator0
        (Utilities|Array|AddUnique (LocalVariables|GetAllUniqueIdentifiersToTest) (Variables|/BRMapCh6/Maps/Test/Chris/BP/BGA_Minigame_Button_CK.BGA_Minigame_Button_CK_C|GetUniqueIdentifierIndex _array_element)))))
  (Utilities|Array|Shuffle (LocalVariables|GetAllUniqueIdentifiersToTest))
  ; Try all buttons and see if you can win
  (bind _array_element_2 (Utilities|Array|ForEachLoopwithBreak (LocalVariables|GetAllUniqueIdentifiersToTest) :!label _label)
    (:LoopBody
      (bind (_succeeded _uniqueidentifier) (CallFunction|TestifIdentifierWillWin _array_element_2 (LocalVariables|GetAllCachedOwnedIndexLocal)))
      (if _succeeded
        (LocalVariables|SetFinalSucceededIdentifierLocal _uniqueidentifier)
        (LocalVariables|SetCombinationsMatchLocal true)
        (goto _label Break)))
    (:Completed
      (return (LocalVariables|GetCombinationsMatchLocal) (LocalVariables|GetFinalSucceededIdentifierLocal)))))

(fn AIStopPlayerWinningMove
  (LocalVariables|SetAllCachedOwnedIndexLocal (Variables|Default|GetAllPlayerOwnedButtonInts)) ; checks for a winning move from the player
  ; Get All Non Pressed Button Indexs
  (for _array_element (Variables|Default|GetAllSpawnedButtonRefs)
    (switch Utilities|FlowControl|Switch|SwitchonEMiniGameCKButtonStates (Variables|/BRMapCh6/Maps/Test/Chris/BP/BGA_Minigame_Button_CK.BGA_Minigame_Button_CK_C|GetButtonState _array_element)
      (:NewEnumerator0
        (Utilities|Array|AddUnique (LocalVariables|GetAllUniqueIdentifiersToTest) (Variables|/BRMapCh6/Maps/Test/Chris/BP/BGA_Minigame_Button_CK.BGA_Minigame_Button_CK_C|GetUniqueIdentifierIndex _array_element)))))
  (Utilities|Array|Shuffle (LocalVariables|GetAllUniqueIdentifiersToTest))
  ; Try all buttons and see if you can win
  (bind _array_element_2 (Utilities|Array|ForEachLoopwithBreak (LocalVariables|GetAllUniqueIdentifiersToTest) :!label _label)
    (:LoopBody
      (bind (_succeeded _uniqueidentifier) (CallFunction|TestifIdentifierWillWin _array_element_2 (LocalVariables|GetAllCachedOwnedIndexLocal)))
      (if _succeeded
        (LocalVariables|SetFinalSucceededIdentifierLocal _uniqueidentifier)
        (LocalVariables|SetCombinationsMatchLocal true)
        (goto _label Break)))
    (:Completed
      (return (LocalVariables|GetCombinationsMatchLocal) (LocalVariables|GetFinalSucceededIdentifierLocal)))))

(fn AIDumbCheck
  (return (>= (Variables|Default|GetAIDumbProbability) (Math|Random|RandomFloatinRange 0.0 1.0))))

(fn AIStopPlayerWinning
  (return false 0))

(fn BlueprintGetInteractionTime
  (params (InteractingPawn :type Fort%20Pawn%20Object%20Reference) (InteractionBeingAttempted :type EInteractionBeingAttempted%20Enum :default "FirstInteraction"))
  (return 0.25 true))

(fn DisableAllButtonInteracts
  (for _array_element (Variables|Default|GetAllSpawnedButtonRefs)
    (Variables|/BRMapCh6/Maps/Test/Chris/BP/BGA_Minigame_Button_CK.BGA_Minigame_Button_CK_C|SetAllowPlayersInteract false _array_element)))

(fn FindButtonFromIdentifier
  (params (IdentifierIndex :type Integer))
  (LocalVariables|SetIndexLocal IdentifierIndex)
  (bind _array_element (Utilities|Array|ForEachLoopwithBreak (Variables|Default|GetAllSpawnedButtonRefs) :!label _label)
    (:LoopBody
      (bind _output_get (LocalVariables|SetCurrentButtonRef _array_element))
      (if (== (LocalVariables|GetIndexLocal) (Variables|/BRMapCh6/Maps/Test/Chris/BP/BGA_Minigame_Button_CK.BGA_Minigame_Button_CK_C|GetUniqueIdentifierIndex _output_get))
        (LocalVariables|SetFoundMatchLocal true)
        (goto _label Break)))
    (:Completed
      (return (LocalVariables|GetCurrentButtonRef)))))

(fn FindRandomButton
  (bind _allavaliablebuttons (CallFunction|ListAllRemainingNonPlayedButtons))
  (Utilities|Array|Shuffle _allavaliablebuttons)
  (return (Variables|/BRMapCh6/Maps/Test/Chris/BP/BGA_Minigame_Button_CK.BGA_Minigame_Button_CK_C|GetUniqueIdentifierIndex (Utilities|Array|Get(acopy) _allavaliablebuttons))))

(fn GameVictoryCheck
  (params (DidPlayerWinRound :type Boolean))
  (bind _output_get (LocalVariables|SetPlayerWinLocal DidPlayerWinRound))
  (if _output_get
    (if (>= (Variables|Default|GetPlayerGameScore) 2)
      (LocalVariables|SetGameVictoryLocal true)
      (else
        (LocalVariables|SetGameVictoryLocal false)))
    (elif (>= (Variables|Default|GetAIGameScore) 2)
      (LocalVariables|SetGameVictoryLocal true)
      (else
        (LocalVariables|SetGameVictoryLocal false))))
  (return (LocalVariables|GetGameVictoryLocal) (LocalVariables|GetPlayerWinLocal)))

(fn GetAllButtonLocations
  (Utilities|FlowControl|Sequence
    (:then_0
      (Utilities|Array|Add (Variables|Default|GetAllButtonTransforms) (Transformation|GetWorldTransform (Variables|Default|GetA1))))
    (:then_1
      (Utilities|Array|Add (Variables|Default|GetAllButtonTransforms) (Transformation|GetWorldTransform (Variables|Default|GetA2))))
    (:then_2
      (Utilities|Array|Add (Variables|Default|GetAllButtonTransforms) (Transformation|GetWorldTransform (Variables|Default|GetA3))))
    (:then_3
      (Utilities|Array|Add (Variables|Default|GetAllButtonTransforms) (Transformation|GetWorldTransform (Variables|Default|GetB1))))
    (:then_4
      (Utilities|Array|Add (Variables|Default|GetAllButtonTransforms) (Transformation|GetWorldTransform (Variables|Default|GetB2))))
    (:then_5
      (Utilities|Array|Add (Variables|Default|GetAllButtonTransforms) (Transformation|GetWorldTransform (Variables|Default|GetB3))))
    (:then_6
      (Utilities|Array|Add (Variables|Default|GetAllButtonTransforms) (Transformation|GetWorldTransform (Variables|Default|GetC1))))
    (:then_7
      (Utilities|Array|Add (Variables|Default|GetAllButtonTransforms) (Transformation|GetWorldTransform (Variables|Default|GetC2))))
    (:then_8
      (Utilities|Array|Add (Variables|Default|GetAllButtonTransforms) (Transformation|GetWorldTransform (Variables|Default|GetC3))))))

(fn IncrementRoundScore
  (params (PlayerWon :type Boolean))
  (if PlayerWon
    (Math|Integer|IncrementInt (Variables|Default|GetPlayerGameScore))
    (else
      (Math|Integer|IncrementInt (Variables|Default|GetAIGameScore)))))

(fn IncrementTurns
  (bind _result (Math|Integer|IncrementInt (Variables|Default|GetCurrentTurns))
    (:%20%20
      (Development|PrintString (Utilities|String|Append "TURN = " (Utilities|String|ToString(Integer) _result)) true true "(R=0.000000,G=0.660000,B=1.000000,A=1.000000)" 6e+01))))

(fn InitEmotionDMI
  (bind _returnvalue (Rendering|Material|CreateDynamicMaterialInstance (Variables|Default|GetFaceMesh) 0 "/BRMapCh6/Maps/Test/Chris/Materials/M_EmotionsMaterialParent.M_EmotionsMaterialParent" "None" :!class /Script/Engine.PrimitiveComponent))
  (Variables|Default|SetEmotionDMI _returnvalue))

(fn ListAllRemainingNonPlayedButtons
  (for _array_element (Variables|Default|GetAllSpawnedButtonRefs)
    (switch Utilities|FlowControl|Switch|SwitchonEMiniGameCKButtonStates (Variables|/BRMapCh6/Maps/Test/Chris/BP/BGA_Minigame_Button_CK.BGA_Minigame_Button_CK_C|GetButtonState _array_element)
      (:NewEnumerator0
        (Utilities|Array|Add (LocalVariables|GetLocalNonPlayedButtons) _array_element))))
  (return (LocalVariables|GetLocalNonPlayedButtons)))

(fn OnRep_AIGameScore
  (Rendering|Components|TextRender|SetText (Variables|Default|GetAIScoreTextRender) (Utilities|Text|ToText(Integer) (Variables|Default|GetAIGameScore))))

(fn OnRep_EFaceStates
  (switch Utilities|FlowControl|Switch|SwitchonEMiniGameCKEmotionStates (Variables|Default|GetEFaceStates)
    (:NewEnumerator0
      (Rendering|Material|SetMaterial (Variables|Default|GetFaceMesh) 0 "/BRMapCh6/Maps/Test/Chris/Materials/M_EmotionsMaterial_Sleeping.M_EmotionsMaterial_Sleeping"))
    (:NewEnumerator1
      (Rendering|Material|SetMaterial (Variables|Default|GetFaceMesh) 0 "/BRMapCh6/Maps/Test/Chris/Materials/M_EmotionsMaterial_Happy.M_EmotionsMaterial_Happy"))
    (:NewEnumerator2
      (Rendering|Material|SetMaterial (Variables|Default|GetFaceMesh) 0 "/BRMapCh6/Maps/Test/Chris/Materials/M_EmotionsMaterial_Sad.M_EmotionsMaterial_Sad"))
    (:NewEnumerator3
      (Rendering|Material|SetMaterial (Variables|Default|GetFaceMesh) 0 "/BRMapCh6/Maps/Test/Chris/Materials/M_EmotionsMaterial_Shifty.M_EmotionsMaterial_Shifty"))
    (:NewEnumerator4
      (Rendering|Material|SetMaterial (Variables|Default|GetFaceMesh) 0 "/BRMapCh6/Maps/Test/Chris/Materials/M_EmotionsMaterial_Shock.M_EmotionsMaterial_Shock"))
    (:NewEnumerator5
      (Rendering|Material|SetMaterial (Variables|Default|GetFaceMesh) 0 "/BRMapCh6/Maps/Test/Chris/Materials/M_EmotionsMaterial_Thinking.M_EmotionsMaterial_Thinking"))))

(fn OnRep_ForceFieldActive
  (bind _forcefieldmesh (Variables|Default|GetForceFieldMesh))
  (if (Variables|Default|GetForceFieldActive)
    (Development|SetHiddeninGame _forcefieldmesh false false)
    (Collision|SetCollisionEnabled _forcefieldmesh "QueryAndPhysics")
    (else
      (Development|SetHiddeninGame (Variables|Default|GetForceFieldMesh) true false)
      (Collision|SetCollisionEnabled (Variables|Default|GetForceFieldMesh) "NoCollision"))))

(fn OnRep_PlayerGameScore
  (Rendering|Components|TextRender|SetText (Variables|Default|GetPlayerScoreTextRender) (Utilities|Text|ToText(Integer) (Variables|Default|GetPlayerGameScore))))

(fn RestTheRound
  (for _array_element (Variables|Default|GetAllSpawnedButtonRefs)
    (Variables|/BRMapCh6/Maps/Test/Chris/BP/BGA_Minigame_Button_CK.BGA_Minigame_Button_CK_C|SetButtonState "NewEnumerator0" _array_element))
  (Utilities|Array|Clear (Variables|Default|GetAllAIOwnedButtonInts))
  (Utilities|Array|Clear (Variables|Default|GetAllPlayerOwnedButtonInts))
  (Variables|Default|SetCurrentPressedButtonIdentifier -1)
  (Variables|Default|SetCurrentTurns 0)
  (CallFunction|SetRandomNeutralEmotion))

(fn RoundVictoryLogic
  (params (IsPlayer :type Boolean) (WinningComboInterger :type Integer))
  (LocalVariables|SetIsPlayerLocal IsPlayer)
  (LocalVariables|SetWinningComboButtonIdentifiers (Utilities|Struct|BreakWinCombinationsMiniGameCk (Utilities|Array|Get(aref) (Variables|Default|GetAllWinningCombinationsStruct) WinningComboInterger)))
  (Development|PrintString (Utilities|String|Append "Winner is Player // " (Utilities|String|ToString(Boolean) IsPlayer)) true true "(R=0.000000,G=0.660000,B=1.000000,A=1.000000)" 32.0)
  (CallFunction|IncrementRoundScore (LocalVariables|GetIsPlayerLocal))
  (for _array_element (Variables|Default|GetAllSpawnedButtonRefs)
    (Variables|/BRMapCh6/Maps/Test/Chris/BP/BGA_Minigame_Button_CK.BGA_Minigame_Button_CK_C|SetAllowPlayersInteract false _array_element)
    (if (Utilities|Array|ContainsItem (LocalVariables|GetWinningComboButtonIdentifiers) (Variables|/BRMapCh6/Maps/Test/Chris/BP/BGA_Minigame_Button_CK.BGA_Minigame_Button_CK_C|GetUniqueIdentifierIndex _array_element))
      (Class|BGAMinigameButtonCK|SetButtonState _array_element "NewEnumerator3" :!class /BRMapCh6/Maps/Test/Chris/BP/BGA_Minigame_Button_CK.BGA_Minigame_Button_CK_C)))
  (if (LocalVariables|GetIsPlayerLocal)
    (CallFunction|VictoryAudio)
    (CallFunction|SetEmotion "NewEnumerator2")
    (else
      (CallFunction|LossAudio)
      (CallFunction|SetEmotion "NewEnumerator1")))
  (return (LocalVariables|GetIsPlayerLocal)))

(fn SetEmotion
  (params (EmotionState :type EMiniGameCKEmotionStates%20Enum :default "NewEnumerator0"))
  (Variables|Default|SetEFaceStates EmotionState))

(fn SpawnButtons
  (CallFunction|GetAllButtonLocations)
  (for (_array_element _array_index) (Variables|Default|GetAllButtonTransforms)
    (LocalVariables|SetCurrentIndex _array_index)
    (bind _returnvalue (Game|SpawnActorfromClass "/BRMapCh6/Maps/Test/Chris/BP/BGA_Minigame_Button_CK.BGA_Minigame_Button_CK_C" _array_element "AlwaysSpawn" "MultiplyWithRoot" 0 0 0))
    (Variables|/BRMapCh6/Maps/Test/Chris/BP/BGA_Minigame_Button_CK.BGA_Minigame_Button_CK_C|SetUniqueIdentifierIndex (LocalVariables|GetCurrentIndex) _returnvalue)
    (Utilities|Array|Add (Variables|Default|GetAllSpawnedButtonRefs) _returnvalue)
    (Variables|/BRMapCh6/Maps/Test/Chris/BP/BGA_Minigame_Button_CK.BGA_Minigame_Button_CK_C|SetOwnerGame self _returnvalue)))

(fn TestIfIdentifierWillWin
  (params (Identifier :type Integer) (AllCachedOwnedIdentifiers :type Array%20of%20Integers))
  (bind _output_get (LocalVariables|SetCachedFullList AllCachedOwnedIdentifiers))
  (Utilities|Array|AddUnique _output_get Identifier)
  (bind (_array_element _array_index) (Utilities|Array|ForEachLoopwithBreak (Variables|Default|GetAllWinningCombinationsStruct) :!label _label)
    (:LoopBody
      (LocalVariables|SetCurrentWinningCombinationInt _array_index)
      (bind _foundcombinationmatch (CallFunction|CheckCombination (Utilities|Struct|BreakWinCombinationsMiniGameCk _array_element) (LocalVariables|GetCachedFullList)))
      (if _foundcombinationmatch
        (LocalVariables|SetCominationsMatchLocal true)
        (LocalVariables|SetFinalMatchingComboInt (LocalVariables|GetCurrentWinningCombinationInt))
        (goto _label Break)))
    (:Completed
      (return (LocalVariables|GetCominationsMatchLocal) Identifier))))

(fn TestVictory
  (params (IsPlayer :type Boolean))
  (bind _output_get (LocalVariables|SetIsPlayerLocal IsPlayer))
  (if _output_get
    (LocalVariables|SetAllOwnedButtonsIntLocal (Variables|Default|GetAllPlayerOwnedButtonInts))
    (else
      (LocalVariables|SetAllOwnedButtonsIntLocal (Variables|Default|GetAllAIOwnedButtonInts))))
  (bind (_array_element _array_index) (Utilities|Array|ForEachLoopwithBreak (Variables|Default|GetAllWinningCombinationsStruct) :!label _label)
    (:LoopBody
      (bind _output_get_2 (LocalVariables|SetCurrentWinningCombinationInt _array_index))
      (bind _foundcombinationmatch (CallFunction|CheckCombination (Utilities|Struct|BreakWinCombinationsMiniGameCk _array_element) (LocalVariables|GetAllOwnedButtonsIntLocal)))
      (if _foundcombinationmatch
        (LocalVariables|SetCominationsMatchLocal true)
        (LocalVariables|SetFinalMatchingComboInt _output_get_2)
        (goto _label Break)))
    (:Completed
      (return (LocalVariables|GetCominationsMatchLocal) (LocalVariables|GetIsPlayerLocal) (LocalVariables|GetFinalMatchingComboInt)))))

(fn ConstructionScript)
