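; Prototype interactable button building actor for a minigame, made invulnerable at BeginPlay with a dynamic material on its cube.
; A short timed interaction passes a DoOnce gate and calls PlayerTurn on the owning game with the button's index.
; SetButtonState sets a RepNotify state that drives material glow and color and allows interaction only in the first state.
; OnRep_ButtonPlayed is an empty stub.
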(event EventBeginPlay
  ; Init
  (Variables|Default|SetAllowPlayersInteract true)
  (CallFunction|CreateDMI)
  (Variables|BuildingActor|SetbIsInvulnerable true))

(event Interaction|EventBlueprintOnInteract
  (params (InteractingPawn :type Fort%20Pawn%20Object%20Reference) (InteractionBeingAttempted :type EInteractionBeingAttempted%20Enum :default "FirstInteraction"))
  ; Player Interaction
  (Utilities|FlowControl|DoOnce false
    (:Completed
      (Class|BGAMiniGameCK|PlayerTurn (Variables|Default|GetOwnerGame) (Variables|Default|GetUniqueIdentifierIndex) self)
      (Utilities|FlowControl|Delay 0.5))))

(event Custom|SetButtonState
  (params (StateToSet :type EMiniGameCKButtonStates%20Enum :default "NewEnumerator0"))
  ; Button State
  (Variables|Default|SetButtonState StateToSet))

(fn BlueprintCanInteract
  (params (InteractingPawn :type Fort%20Pawn%20Object%20Reference) (InteractionBeingAttempted :type EInteractionBeingAttempted%20Enum :default "FirstInteraction") (InteractionType :type TInteractionType%20Enum :default "IT_NoInteraction"))
  (return (Variables|Default|GetAllowPlayersInteract)))

(fn BlueprintGetInteractionString
  (params (InteractingPawn :type Fort%20Pawn%20Object%20Reference) (InteractionBeingAttempted :type EInteractionBeingAttempted%20Enum :default "FirstInteraction"))
  (return (Variables|Default|GetInteractString)))

(fn BlueprintGetInteractionTime
  (params (InteractingPawn :type Fort%20Pawn%20Object%20Reference) (InteractionBeingAttempted :type EInteractionBeingAttempted%20Enum :default "FirstInteraction"))
  (return 0.33 true))

(fn CreateDMI
  (bind _returnvalue (Rendering|Material|CreateDynamicMaterialInstance (Variables|Default|GetCube) 0 "/BRMapCh6/Maps/Test/Chris/Materials/M_Button_Proto.M_Button_Proto" "None" :!class /Script/Engine.PrimitiveComponent))
  (Variables|Default|SetDMI _returnvalue))

(fn OnRep_ButtonPlayed
  (if true
    _))

(fn OnRep_ButtonState
  (switch Utilities|FlowControl|Switch|SwitchonEMiniGameCKButtonStates (Variables|Default|GetButtonState)
    (:NewEnumerator0
      (Rendering|Material|SetScalarParameterValue (Variables|Default|GetDMI) "Glow" 0.25 :!class /Script/Engine.MaterialInstanceDynamic)
      (Variables|Default|SetAllowPlayersInteract true)
      (Rendering|Material|SetVectorParameterValue (Variables|Default|GetDMI) "Color" "(R=0.197917,G=0.197917,B=0.197917,A=1.000000)" :!class /Script/Engine.MaterialInstanceDynamic))
    (:NewEnumerator1
      (Rendering|Material|SetScalarParameterValue (Variables|Default|GetDMI) "Glow" 0.25 :!class /Script/Engine.MaterialInstanceDynamic)
      (Variables|Default|SetAllowPlayersInteract false)
      (Rendering|Material|SetVectorParameterValue (Variables|Default|GetDMI) "Color" "(R=0.008254,G=0.083333,B=0.002644,A=1.000000)" :!class /Script/Engine.MaterialInstanceDynamic))
    (:NewEnumerator2
      (Rendering|Material|SetScalarParameterValue (Variables|Default|GetDMI) "Glow" 0.25 :!class /Script/Engine.MaterialInstanceDynamic)
      (Variables|Default|SetAllowPlayersInteract false)
      (Rendering|Material|SetVectorParameterValue (Variables|Default|GetDMI) "Color" "(R=0.234375,G=0.000819,B=0.011492,A=1.000000)" :!class /Script/Engine.MaterialInstanceDynamic))
    (:NewEnumerator3
      (Rendering|Material|SetScalarParameterValue (Variables|Default|GetDMI) "Glow" 3.0 :!class /Script/Engine.MaterialInstanceDynamic)
      (Variables|Default|SetAllowPlayersInteract false))
    (:NewEnumerator4
      (Rendering|Material|SetScalarParameterValue (Variables|Default|GetDMI) "Glow" 0.0 :!class /Script/Engine.MaterialInstanceDynamic)
      (Variables|Default|SetAllowPlayersInteract false)
      (Rendering|Material|SetVectorParameterValue (Variables|Default|GetDMI) "Color" "(R=0.197917,G=0.197917,B=0.197917,A=1.000000)" :!class /Script/Engine.MaterialInstanceDynamic))))

(fn ConstructionScript)
