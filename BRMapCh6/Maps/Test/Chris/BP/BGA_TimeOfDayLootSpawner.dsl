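; Time-of-day loot spawner actor that spawns loot from a tier group table at its arrow.
; On authority, it waits for the time-of-day manager, then either spawns on selected day phases or polls a time range on a looping timer.
; A timed spawn pauses that timer until the next phase-change callback.
; Without random toss rotation, pickups are tossed along the arrow with random jitter.
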(event Custom|Check%20Time
  (bind _todinterfaceref (Variables|Default|GetTODInterfaceRef))
  (bind _returnvalue (TimeOfDay|GetTimeOfDay _todinterfaceref :!class /Script/FortniteGame.FortTimeOfDayManagerInterface))
  (bind _specifictimedlootrange (Variables|Default|GetSpecificTimedLootRange))
  (bind (_x _y) (Math|Vector2D|BreakVector2D _specifictimedlootrange))
  (if (>= _returnvalue _x)
    (if (<= _returnvalue _y)
      (Utilities|FlowControl|Sequence
        (:then_0
          (Utilities|Time|PauseTimerbyHandle (Variables|Default|GetLoopingTimer))
          (TimeOfDay|RegisterforDayPhaseChangedEvent (Variables|Default|GetTODInterfaceRef) (delegate Custom|CurrentPhase) false))
        (:then_1
          (bind _itempickupsspawned (Class|AthenaFunctionLibrary|SpawnLootfromTable (Variables|Default|GetTierGroupName) 1 -1 (Transformation|GetWorldLocation (Variables|Default|GetArrow)) (Transformation|GetForwardVector (Variables|Default|GetArrow)) 0 true (Variables|Default|GetRandomTossRotation)))
          (if (not (Variables|Default|GetRandomTossRotation))
            (for _array_element _itempickupsspawned
              (Pickup|TossPickupinDirection _array_element (+ (* (Transformation|GetForwardVector (Variables|Default|GetArrow)) (Math|Conversions|ToVector(Float) (Variables|Default|GetDirectionTossMagnitude))) (Math|Vector|MakeVector (Math|Random|RandomFloatinRange -1.5e+02 1.5e+02) (Math|Random|RandomFloatinRange -1.5e+02 1.5e+02) (Math|Random|RandomFloatinRange 25.0 1.5e+02))) self false false))))))))

(event EventBeginPlay
  (Networking|SwitchHasAuthority
    (:Authority
      (Utilities|FlowControl|Delay 2.0)
      (bind _timeofdaymanager (Class|FortAsyncActionTimeOfDayManagerReady|GetTimeOfDayManagerReady)
        (:OnReady
          (if (Utilities|IsValidInterface _timeofdaymanager)
            _
            (else
              _))
          (Variables|Default|SetTODInterfaceRef _timeofdaymanager)
          (if (Variables|Default|GetUsePhaseSpawning)
            (TimeOfDay|RegisterforDayPhaseChangedEvent _timeofdaymanager (delegate Custom|Phase) true)
            (else
              (bind _returnvalue (Utilities|Time|SetTimerbyEvent (delegate Custom|Check%20Time) 5.0 true))
              (Variables|Default|SetLoopingTimer _returnvalue))))))))

(event Custom|CurrentPhase
  (params (CurrentDayPhase :type EFortDayPhase%20Enum :default "Morning") (PreviousDayPhase :type EFortDayPhase%20Enum :default "Morning") (bAtCreation :type Boolean))
  (Utilities|Time|UnpauseTimerbyHandle (Variables|Default|GetLoopingTimer)))

(event Custom|Phase
  (params (CurrentDayPhase :type EFortDayPhase%20Enum :default "Morning") (PreviousDayPhase :type EFortDayPhase%20Enum :default "Morning") (bAtCreation :type Boolean))
  (if (Utilities|Array|ContainsItem (Variables|Default|GetPhasesToSpawnLoot) CurrentDayPhase)
    (bind _itempickupsspawned (Class|AthenaFunctionLibrary|SpawnLootfromTable (Variables|Default|GetTierGroupName) 1 -1 (Transformation|GetWorldLocation (Variables|Default|GetArrow)) (Transformation|GetForwardVector (Variables|Default|GetArrow)) 0 true (Variables|Default|GetRandomTossRotation)))
    (if (not (Variables|Default|GetRandomTossRotation))
      (for _array_element _itempickupsspawned
        (Pickup|TossPickupinDirection _array_element (+ (* (Transformation|GetForwardVector (Variables|Default|GetArrow)) (Math|Conversions|ToVector(Float) (Variables|Default|GetDirectionTossMagnitude))) (Math|Vector|MakeVector (Math|Random|RandomFloatinRange -1.5e+02 1.5e+02) (Math|Random|RandomFloatinRange -1.5e+02 1.5e+02) (Math|Random|RandomFloatinRange 25.0 1.5e+02))) self false false)))))

(event Custom|LootSpawnAudio
  (Networking|SwitchHasAuthority
    (:Remote
      (Audio|SpawnSoundatLocation (Variables|Default|GetLootSpawnAudioCue) (Transformation|GetWorldLocation (Variables|Default|GetArrow))))))

(fn ConstructionScript)
