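; Test egg actor that hatches when shot, destroyed or approached by a non-crouching pawn, spawning an archery target as a placeholder when killed.
; A server-side sphere overlap with a repeating timer checks crouch state and calls HatchEgg.
; HatchEgg runs a random-speed scale timeline, then triggers a chain reaction in nearby unhatched eggs and destroys the actor.
; RepNotifies handle collision and visibility, while remote-only branches play sounds and burst effects.
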(event EventBeginPlay
  (bind _eggmesh (Variables|Default|GetEggMesh))
  (bind _returnvalue (Transformation|GetWorldScale _eggmesh))
  ; Init World Scale of Eggs
  (Utilities|FlowControl|Delay 0.2)
  (Variables|Default|SetEggDefaultScale _returnvalue)
  (Variables|Default|SetEggHatchScale (Math|Vector|MakeVector (+ (.x _returnvalue) 0.45) (+ (.y _returnvalue) 0.45) (+ (.z _returnvalue) 0.55)))
  (EventDispatchers|BindEventtoOnDied (delegate Custom|EggDied)))

(event Custom|EggDied
  (params (DamagedActor :type Actor%20Object%20Reference) (Damage :type Float%20%28single-precision%29) (InstigatedBy :type Controller%20Object%20Reference) (DamageCauser :type Actor%20Object%20Reference) (HitLocation :type Vector) (FHitComponent :type Primitive%20Component%20Object%20Reference) (BoneName :type Name) (Momentum :type Vector))
  ; Init World Scale of Eggs
  (Variables|Default|SetEggWasDestroyed true))

(event EventOnDeathServer
  (params (Damage :type Float%20%28single-precision%29) (DamageTags :type Gameplay%20Tag%20Container%20Structure%20%28by%20ref%29) (Momentum :type Vector :default "0, 0, 0") (HitInfo :type Hit%20Result%20Structure%20%28by%20ref%29) (InstigatedBy :type Controller%20Object%20Reference) (DamageCauser :type Actor%20Object%20Reference) (EffectContext :type Gameplay%20Effect%20Context%20Handle%20Structure))
  ; Damage from players
  (Utilities|FlowControl|DoOnce false
    (:Completed
      (CallFunction|HideEggmeshesonClient)
      (CallFunction|HatchEgg)
      (Game|SpawnActorfromClass "/Game/Environments/Helios/Props/Outdoor/Outdoor_ArcheryTarget_A/Blueprints/BP_Outdoor_ArcheryTarget_A.BP_Outdoor_ArcheryTarget_A_C" (Math|Transform|MakeTransform (Transformation|GetWorldLocation (Variables|Default|GetDefaultSceneRoot)) (Math|Rotator|MakeRotator 0.0 0.0 (Math|Random|RandomFloatinRange -179.0 179.0))) "AlwaysSpawn" "MultiplyWithRoot" 0 0 0 0) ; spawn loot, or a monster here?
      (CallFunction|ClientDeathEffects))))

(event Custom|HatchEgg
  ; Egg Hatch Animation
  (Utilities|FlowControl|DoOnce false
    (:Completed
      (Variables|Default|SetEggHatchingInitiated true)
      (CallFunction|EggHatchInitClientEffects)
      (Components|Timeline|SetPlayRate (Variables|BP_EggTestCK|GetTimeline) (Math|Random|RandomFloatinRange 0.7 1.3))
      (bind (_direction _egghatch) (Timeline|Timeline PlayFromStart)
        (:Update
          (Transformation|SetWorldScale3D (Variables|Default|GetEggMesh) (Math|Vector|Lerp(Vector) (Variables|Default|GetEggDefaultScale) (Variables|Default|GetEggHatchScale) _egghatch)))
        (:Finished
          (CallFunction|CheckNearbyEggsonHatching)
          (Actor|DestroyActor))))))

(event Custom|EggHatchInitClientEffects
  ; Egg Hatch Animation
  (if (not (Variables|Default|GetEggWasDestroyed))
    (Networking|SwitchHasAuthority
      (:Remote
        (Audio|SpawnSoundatLocation "/Game/Sounds/Fort_GamePlay_Sounds/Player/RiftPortal/RiftPortal_Expand_Cue.RiftPortal_Expand_Cue" (Transformation|GetWorldLocation (Variables|Default|GetEggMesh)))))))

(event Custom|HideEggmeshesOnClient
  ; Egg visibility on death
  (Variables|Default|SetEggVisibilityState false))

(event Custom|ClientDeathEffects
  ; Egg destruction FX
  (Networking|SwitchHasAuthority
    (:Remote
      (Components|Activation|Activate (Variables|Default|GetNS_Flopper_JellyFish_Explode) false)
      (Audio|SpawnSoundatLocation "/Game/Athena/Sounds/Items/Fizz/Paprika_Fizz_Explode_Cue.Paprika_Fizz_Explode_Cue" (Transformation|GetWorldLocation (Variables|Default|GetEggMesh))))))

(event OnComponentBeginOverlap(Sphere)
  (params (OverlappedComponent :type Primitive%20Component%20Object%20Reference) (OtherActor :type Actor%20Object%20Reference) (OtherComp :type Primitive%20Component%20Object%20Reference) (OtherBodyIndex :type Integer) (bFromSweep :type Boolean) (SweepResult :type Hit%20Result%20Structure%20%28by%20ref%29))
  ; Egg proximity Check
  (Networking|SwitchHasAuthority
    (:Authority
      (if (Utilities|Time|IsTimerActivebyHandle (Variables|Default|GetProximityTimer))
        _
        (elif (Variables|Default|GetEggHatchingInitiated)
          _
          (else
            (bind _returnvalue (Utilities|Time|SetTimerbyEvent (delegate Custom|ProximityTrigger) 1.0 true))
            (Variables|Default|SetProximityTimer _returnvalue)
            (bind _outactors (Collision|SphereOverlapActors (Transformation|GetWorldLocation (Variables|Default|GetSphere)) (Variables|/Script/Engine.SphereComponent|GetSphereRadius (Variables|Default|GetSphere)) (Utilities|Array|MakeArray "ObjectTypeQuery3") "/Script/FortniteGame.FortPawn")) ; bots and animals will trigger there, could change to just players if needed (fortathenaplayerpawn)
            (bind _output_get (Variables|Default|SetAllPawnActorsOverlapping _outactors))
            (if (> (Utilities|Array|Length _output_get) 0)
              (bind _playersareallcrouching (CallFunction|AreAllPlayersCrouching (Variables|Default|GetAllPawnActorsOverlapping)))
              (if (not _playersareallcrouching)
                (Utilities|FlowControl|DoOnce false
                  (:Completed
                    (Utilities|Time|PauseTimerbyHandle (Variables|Default|GetProximityTimer))
                    (CallFunction|HatchEgg)))))))))))

(event OnComponentEndOverlap(Sphere)
  (params (OverlappedComponent :type Primitive%20Component%20Object%20Reference) (OtherActor :type Actor%20Object%20Reference) (OtherComp :type Primitive%20Component%20Object%20Reference) (OtherBodyIndex :type Integer))
  ; Egg proximity Check
  (Networking|SwitchHasAuthority
    (:Authority
      (Utilities|Time|PauseTimerbyHandle (Variables|Default|GetProximityTimer)))))

(event Custom|ProximityTrigger
  ; Egg proximity Check
  (bind _outactors (Collision|SphereOverlapActors (Transformation|GetWorldLocation (Variables|Default|GetSphere)) (Variables|/Script/Engine.SphereComponent|GetSphereRadius (Variables|Default|GetSphere)) (Utilities|Array|MakeArray "ObjectTypeQuery3") "/Script/FortniteGame.FortPawn")) ; bots and animals will trigger there, could change to just players if needed (fortathenaplayerpawn)
  (bind _output_get (Variables|Default|SetAllPawnActorsOverlapping _outactors))
  (if (> (Utilities|Array|Length _output_get) 0)
    (bind _playersareallcrouching (CallFunction|AreAllPlayersCrouching (Variables|Default|GetAllPawnActorsOverlapping)))
    (if (not _playersareallcrouching)
      (Utilities|FlowControl|DoOnce false
        (:Completed
          (Utilities|Time|PauseTimerbyHandle (Variables|Default|GetProximityTimer))
          (CallFunction|HatchEgg))))))

(event EventOnDamageServer
  (params (Damage :type Float%20%28single-precision%29) (DamageTags :type Gameplay%20Tag%20Container%20Structure%20%28by%20ref%29) (Momentum :type Vector :default "0, 0, 0") (HitInfo :type Hit%20Result%20Structure%20%28by%20ref%29) (InstigatedBy :type Controller%20Object%20Reference) (DamageCauser :type Actor%20Object%20Reference) (EffectContext :type Gameplay%20Effect%20Context%20Handle%20Structure))
  ; Damaged to Egg start to Hatch
  (Utilities|FlowControl|DoOnce false
    (:Completed
      (CallFunction|HatchEgg))))

(event Custom|CheckNearbyEggsOnHatching
  ; hatch Nearby Eggs
  (Utilities|FlowControl|Delay (Math|Random|RandomFloatinRange 0.0 0.75))
  (bind _outactors (Collision|SphereOverlapActors (Transformation|GetWorldLocation (Variables|Default|GetSphere)) (* (Variables|/Script/Engine.SphereComponent|GetSphereRadius (Variables|Default|GetSphere)) 1.0) (Utilities|Array|MakeArray "ObjectTypeQuery2") "/BRMapCh6/Maps/Test/Chris/BP/BP_EggTestCK.BP_EggTestCK_C" (Utilities|Array|MakeArray self))) ; get all nearby eggs 
  (Variables|Default|SetAllNearbyEggs _outactors)
  (bind (_valideggstohatch _foundvalideggs) (CallFunction|FindAValidEgg (Variables|Default|GetAllNearbyEggs))) ; filter eggs that are still valid, have not hatched yet, and are close enough to hatch
  (if _foundvalideggs
    (bind _output_get (Variables|Default|SetAllValidNearbyEggs _valideggstohatch))
    (if (> (Utilities|Array|Length _output_get) 1)
      (switch int (Math|Random|RandomIntegerinRange 0 2)
        (:0
          _)
        (:1
          _)
        (:2
          (for _index (range 2)
            (bind _asbp_egg_test_ck (Utilities|Casting|CastToBP_EggTestCK (Utilities|Array|Get(acopy) _output_get _index)))
            (CallFunction|HatchEgg _asbp_egg_test_ck))))
      (else
        _))
    (Utilities|Array|Shuffle _output_get) ; hatch just one nearby egg
    (bind _asbp_egg_test_ck_2 (Utilities|Casting|CastToBP_EggTestCK (Utilities|Array|Get(acopy) _output_get)))
    (CallFunction|HatchEgg _asbp_egg_test_ck_2)))

(fn AreAllPlayersCrouching
  (params (Actors :type Array%20of%20Actor%20Object%20References))
  (bind _array_element (Utilities|Array|ForEachLoopwithBreak Actors)
    (:LoopBody
      (Utilities|IsValid _array_element
        (:Is%20Valid
          (bind _asfort_pawn (Utilities|Casting|CastToFortPawn _array_element))
          (if (not (Variables|/Script/Engine.Character|GetbIsCrouched _asfort_pawn))
            (LocalVariables|SetAllThePawnsAreCrouching false)))))
    (:Completed
      _))
  (return (LocalVariables|GetAllThePawnsAreCrouching)))

(fn FindAValidEgg
  (params (AllNearbyEggs :type Array%20of%20Actor%20Object%20References))
  (for _array_element AllNearbyEggs
    (Utilities|IsValid _array_element
      (:Is%20Valid
        (bind _asbp_egg_test_ck (Utilities|Casting|CastToBP_EggTestCK _array_element))
        (if (not (Variables|Default|GetEggHatchingInitiated _asbp_egg_test_ck))
          (bind _distance (CallFunction|GetNearbyEggDistance _asbp_egg_test_ck)) ; check the egg root not the bounds inc the sphere
          (if (<= _distance (Variables|/Script/Engine.SphereComponent|GetSphereRadius (Variables|Default|GetSphere)))
            (Utilities|Array|Add (LocalVariables|GetValidEggsLocal) _asbp_egg_test_ck))))))
  (return (LocalVariables|GetValidEggsLocal) (> (Utilities|Array|Length (LocalVariables|GetValidEggsLocal)) 0)))

(fn GetNearbyEggDistance
  (params (NearbyEgg :type BP%20Egg%20Test%20CK%20Object%20Reference))
  (return (Math|Float|Absolute(Float) (Math|Vector|VectorLength (- (Transformation|GetWorldLocation (Variables|Default|GetDefaultSceneRoot NearbyEgg)) (Transformation|GetActorLocation self))))))

(fn OnRep_EggHatchingInitiated
  (if (Variables|Default|GetEggHatchingInitiated)
    (Collision|SetCollisionEnabled (Variables|Default|GetEgg_A_Base) "NoCollision") ; turn collision off when animating
    (else
      (Collision|SetCollisionEnabled (Variables|Default|GetEgg_A_Base) "QueryAndPhysics"))))

(fn OnRep_EggVisibilityState
  (if (Variables|Default|GetEggVisibilityState)
    (Rendering|SetVisibility (Variables|Default|GetEggMesh) true false)
    (Collision|SetCollisionEnabled (Variables|Default|GetEggMesh) "QueryAndPhysics")
    (else
      (Rendering|SetVisibility (Variables|Default|GetEggMesh) false false)
      (Collision|SetCollisionEnabled (Variables|Default|GetEggMesh) "NoCollision"))))

(fn ConstructionScript
  (bind _sphere (Variables|Default|GetSphere))
  (Transformation|SetWorldScale3D _sphere "1.000000,1.000000,1.000000") ; allow designer to scale the egg without effecting the overlap radius
  (Components|Sphere|SetSphereRadius _sphere (Variables|/Script/Engine.SphereComponent|GetSphereRadius _sphere) true)
  (if (Variables|Default|GetRandomEggYawRotation)
    (Transformation|SetRelativeRotation (Variables|Default|GetEggMesh) (Math|Rotator|MakeRotator 0.0 0.0 (Math|Random|RandomFloatinRange -179.0 179.0)) false false)
    (else
      (Transformation|SetRelativeRotation (Variables|Default|GetEggMesh) "0, 0, 0" false false))))
