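; Prototype horizontal shooting target for a target minigame that slides its mesh back and forth along local X.
; On authority, it becomes invulnerable, binds damage and computes the end offset, with an optional direction flip.
; A state enum drives one timeline forward or in reverse, switching after random pauses until the manager reports game over.
; A separate timeline returns it to start, and hits in one state call an empty TargetHit event.
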(event EventBeginPlay
  ; Init
  (Networking|SwitchHasAuthority
    (:Authority
      (Variables|BuildingActor|SetbIsInvulnerable true)
      (EventDispatchers|BindEventtoOnDamaged self (delegate Custom|Hit))
      (Variables|Default|SetCurrentTargetState "NewEnumerator0")
      (bind _output_get (Variables|Default|SetRelativeStartLocation (Variables|/Script/Engine.SceneComponent|GetRelativeLocation (Variables|Default|GetTargetMesh_0))))
      (if (Variables|Default|GetFlipDirection)
        (Variables|Default|SetRelativeEndLocation (Math|Vector|MakeVector (+ (.x _output_get) -7e+02) (.y _output_get) (.z _output_get)))
        (else
          (Variables|Default|SetRelativeEndLocation (Math|Vector|MakeVector (+ (.x _output_get) 7e+02) (.y _output_get) (.z _output_get))))))))

(event Custom|Hit
  (params (DamagedActor :type Actor%20Object%20Reference) (Damage :type Float%20%28single-precision%29) (InstigatedBy :type Controller%20Object%20Reference) (DamageCauser :type Actor%20Object%20Reference) (HitLocation :type Vector) (FHitComponent :type Primitive%20Component%20Object%20Reference) (BoneName :type Name) (Momentum :type Vector))
  ; Init
  (switch Utilities|FlowControl|Switch|SwitchonTargetMinigameTargetState (Variables|Default|GetCurrentTargetState)
    (:NewEnumerator2
      (CallFunction|TargetHit))))

(event Custom|InitTargetState
  (params (TargetState :type TargetMinigameTargetState%20Enum :default "NewEnumerator0"))
  ; Set State
  (Networking|SwitchHasAuthority
    (:Authority
      (bind _output_get (Variables|Default|SetCurrentTargetState TargetState))
      (switch Utilities|FlowControl|Switch|SwitchonTargetMinigameTargetState _output_get
        (:NewEnumerator0
          (Development|PrintString "IDLE" true true "(R=0.000000,G=0.660000,B=1.000000,A=1.000000)" 1e+01))
        (:NewEnumerator1
          (CallFunction|Deploying))
        (:NewEnumerator2
          (CallFunction|Retracting))
        (:NewEnumerator4
          (CallFunction|Deploying))))))

(event Custom|Deploying
  ; Deploy Target
  (bind (_direction _float) (Timeline|RaiseTarget PlayFromStart)
    (:Update
      (Variables|Default|SetIsTargetMoving true)
      (Transformation|SetRelativeLocation (Variables|Default|GetTargetMesh_0) (Math|Vector|Lerp(Vector) (Variables|Default|GetRelativeStartLocation) (Variables|Default|GetRelativeEndLocation) _float) false false))
    (:Finished
      (Variables|Default|SetIsTargetMoving false)
      (Utilities|FlowControl|Delay (Math|Random|RandomFloatinRange 0.25 1.5))
      (if (not (Variables|/BRMapCh6/Maps/Test/Chris/BP/BP_TargetMinigame_Manager.BP_TargetMinigame_Manager_C|GetGameOver (Variables|Default|GetManagerRef)))
        (switch Utilities|FlowControl|Switch|SwitchonTargetMinigameTargetState (Variables|Default|GetCurrentTargetState)
          (:NewEnumerator1
            (CallFunction|InitTargetState "NewEnumerator2"))
          (:NewEnumerator2
            (CallFunction|InitTargetState "NewEnumerator4"))
          (:NewEnumerator4
            (CallFunction|InitTargetState "NewEnumerator2")))))))

(event Custom|Retracting
  ; Deploy Target
  (Timeline|RaiseTarget ReverseFromEnd))

(event Custom|ResetTargetToStart
  ; Reset Location
  (if (Variables|Default|GetIsTargetMoving)
    (Utilities|FlowControl|Delay 1.0))
  (Variables|Default|SetCurrentRelativeLocation (Variables|/Script/Engine.SceneComponent|GetRelativeLocation (Variables|Default|GetTargetMesh_0)))
  (bind (_direction _float) (Timeline|Reset PlayFromStart)
    (:Update
      (Variables|Default|SetIsTargetMoving true)
      (Transformation|SetRelativeLocation (Variables|Default|GetTargetMesh_0) (Math|Vector|Lerp(Vector) (Variables|Default|GetCurrentRelativeLocation) (Variables|Default|GetRelativeStartLocation) _float) false false))))

(event Custom|TargetHit)

(fn ConstructionScript)
