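; Test-map trigger volume that fires a one-time water explosion effect when a player pawn enters it while grinding.
; On authority, box overlap prints the pawn's gameplay tags, then calls ExecuteEvents through a DoOnce if MovementMode.Grinding is present, otherwise prints "Hello".
; ExecuteEvents spawns the Niagara effect at the Billboard only on its Remote branch.
; The construction script sizes the box from BoxExtents.
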(event Collision|EventActorBeginOverlap
  (params (OtherActor :type Actor%20Object%20Reference)))

(event EventTick
  (params (DeltaSeconds :type Float%20%28single-precision%29)))

(event OnComponentBeginOverlap(Collision)
  (params (OverlappedComponent :type Primitive%20Component%20Object%20Reference) (OtherActor :type Actor%20Object%20Reference) (OtherComp :type Primitive%20Component%20Object%20Reference) (OtherBodyIndex :type Integer) (bFromSweep :type Boolean) (SweepResult :type Hit%20Result%20Structure%20%28by%20ref%29))
  (Networking|SwitchHasAuthority
    (:Authority
      (bind _asfort_player_controller_athena (Utilities|Casting|CastToFortPlayerControllerAthena (Fortnite|GetControllerfromActor OtherActor)))
      (bind _asfort_player_pawn_athena (Utilities|Casting|CastToFortPlayerPawnAthena (Game|Player|GetPlayerPawn _asfort_player_controller_athena)))
      (for _array_element (GameplayTags|BreakGameplayTagContainer (Variables|/Script/FortniteGame.FortPawn|GetGameplayTags _asfort_player_pawn_athena))
        (Development|PrintString (GameplayTags|GetDebugStringfromGameplayTag _array_element) true true "(R=0.000000,G=0.660000,B=1.000000,A=1.000000)" 2e+01))
      (if (Utilities|Array|ContainsItem (GameplayTags|BreakGameplayTagContainer (Variables|/Script/FortniteGame.FortPawn|GetGameplayTags _asfort_player_pawn_athena)) "(TagName=\"MovementMode.Grinding\")")
        (Utilities|FlowControl|DoOnce false
          (:Completed
            (CallFunction|ExecuteEvents)))
        (else
          (Development|PrintString "Hello"))))))

(event Custom|ResetEvents
  (Utilities|FlowControl|DoOnce false
    (:Completed
      (CallFunction|ExecuteEvents))))

(event Custom|ExecuteEvents
  (Networking|SwitchHasAuthority
    (:Remote
      (Niagara|SpawnSystematLocation "/Game/Athena/Items/Consumables/AirStrike/NS_Applesauce_Explosion_Water.NS_Applesauce_Explosion_Water" (Transformation|GetWorldLocation (Variables|Default|GetBillboard))))))

(fn ConstructionScript
  (Components|Box|SetBoxExtent (Variables|Default|GetCollision) (Variables|Default|GetBoxExtents) true))
