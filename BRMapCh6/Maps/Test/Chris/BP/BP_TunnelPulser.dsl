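; BP_TunnelPulser is a test-map actor that shakes nearby players' cameras and destroys player-built structures around it.
; Looping sub-second timers run sphere overlaps around the actor, split by network authority.
; The authority side applies massive environment damage to player-built walls, roofs and floors, reusing one stored overlap result across three cast loops.
; Remote instances find overlapping player pawns and start a low-scale, camera-local Zero Point shake on their controllers.
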(event Custom|camshake
  (bind _outactors (Collision|SphereOverlapActors (Transformation|GetWorldLocation (Variables|Default|GetDefaultSceneRoot)) 1.5e+03 (Utilities|Array|MakeArray "ObjectTypeQuery3") "/Script/FortniteGame.FortPlayerPawnAthena"))
  (for _array_element _outactors
    (bind _asfort_player_controller_athena (Utilities|Casting|CastToFortPlayerControllerAthena (Fortnite|GetControllerfromActor _array_element)))
    (Game|Feedback|ClientStartCameraShake :self _asfort_player_controller_athena :Shake "/FNGameplayCues/Athena/Environments/Nexus/Blueprints/B_CameraShake_ZeroPoint2_Enter.B_CameraShake_ZeroPoint2_Enter_C" :Scale 0.12 :PlaySpace "CameraLocal")))

(event EventBeginPlay
  (Networking|SwitchHasAuthority
    (:Authority
      (Utilities|Time|SetTimerbyEvent (delegate Custom|Destroystuff) 0.225 true))
    (:Remote
      (Utilities|Time|SetTimerbyEvent (delegate Custom|camshake) 0.225 true))))

(event Custom|Destroystuff
  (bind _outactors (Collision|SphereOverlapActors (Transformation|GetWorldLocation (Variables|Default|GetDefaultSceneRoot)) 1.5e+03 (Utilities|Array|MakeArray "ObjectTypeQuery1") "/Script/FortniteGame.BuildingActor" (Utilities|Array|MakeArray self)))
  (bind _output_get (Variables|Default|SetOverlapResults _outactors))
  (for _array_element _output_get
    (bind _asbuilding_wall (Utilities|Casting|CastToBuildingWall _array_element))
    (if (BuildingActor|IsPlayerBuilt _asbuilding_wall)
      (Environment|ApplyGlobalEnvironmentDamagetoActor _asbuilding_wall 9999999.0)))
  (for _array_element_2 (Variables|Default|GetOverlapResults)
    (bind _asbuilding_roof (Utilities|Casting|CastToBuildingRoof _array_element_2))
    (if (BuildingActor|IsPlayerBuilt _asbuilding_roof)
      (Environment|ApplyGlobalEnvironmentDamagetoActor _asbuilding_roof 9999999.0)))
  (for _array_element_3 (Variables|Default|GetOverlapResults)
    (bind _asbuilding_floor (Utilities|Casting|CastToBuildingFloor _array_element_3))
    (if (BuildingActor|IsPlayerBuilt _asbuilding_floor)
      (Environment|ApplyGlobalEnvironmentDamagetoActor _asbuilding_floor 9999999.0))))

(fn ConstructionScript)
