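; Water jet launcher in a test map folder that impulses physics test tubes or launches player pawns, usually along its arrow.
; Always-active mode acts on box overlap on the authority.
; Looping mode deactivates the standing Niagara effect and runs an authority timer that box-checks tubes, then players in a larger box only if none are found, with a vertical-launch variant.
; Each pulse calls ClientFX, whose remote branch spawns a short-lived waterfall effect.
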(event OnComponentBeginOverlap(Box)
  (params (OverlappedComponent :type Primitive%20Component%20Object%20Reference) (OtherActor :type Actor%20Object%20Reference) (OtherComp :type Primitive%20Component%20Object%20Reference) (OtherBodyIndex :type Integer) (bFromSweep :type Boolean) (SweepResult :type Hit%20Result%20Structure%20%28by%20ref%29))
  (bind _arrow (Variables|Default|GetArrow))
  (bind _returnvalue (Transformation|GetForwardVector _arrow))
  (bind _launchvelocity (Variables|Default|GetLaunchVelocity))
  (bind _returnvalue_2 (* _returnvalue _launchvelocity))
  ; Always Active Trigger
  (if (not (Variables|Default|GetUseLoopingActivation))
    (Networking|SwitchHasAuthority
      (:Authority
        (bind _asbga_physics_test_tube (Utilities|Casting|CastToBGA_PhysicsTestTube OtherActor)
          (:then
            (FortPhysicsSimulationLibrary|AddPhysicsImpulseatLocation _asbga_physics_test_tube (* _returnvalue_2 (Variables|Default|GetLaunchMagnitudeScalar)) (Transformation|GetWorldLocation _arrow) true "None" true 0 "(MaxSimSizeAllowed=Large,bRejectEmbeddedInLandscape=True)"))
          (:CastFailed
            (bind _asfort_player_pawn_athena (Utilities|Casting|CastToFortPlayerPawnAthena OtherActor))
            (Character|LaunchCharacter _asfort_player_pawn_athena (* _returnvalue_2 (Variables|Default|GetLaunchMagnitudeScalar)) true true)))))))

(event EventBeginPlay
  ; Timed Activation
  (if (Variables|Default|GetUseLoopingActivation)
    (Components|Activation|Deactivate (Variables|Default|GetNiagara))
    (Networking|SwitchHasAuthority
      (:Authority
        (Utilities|Time|SetTimerbyEvent (delegate Custom|CheckOverlaps) (Variables|Default|GetLoopingActivationTime) true)))))

(event Custom|CheckOverlaps
  ; Timed Activation
  (CallFunction|ClientFX)
  (bind _outactors (Collision|BoxOverlapActors (.location (Transformation|GetWorldTransform (Variables|Default|GetBox))) (Variables|/Script/Engine.BoxComponent|GetBoxExtent (Variables|Default|GetBox)) (Utilities|Array|MakeArray "ObjectTypeQuery1") "/BRMapCh6/Maps/Test/Chris/BP/BGA_PhysicsTestTube.BGA_PhysicsTestTube_C" :!class /Script/Engine.KismetSystemLibrary))
  (if (> (Utilities|Array|Length _outactors) 0)
    (for _array_element _outactors
      (if (Variables|Default|GetisVerticalLauncher)
        (bind _asbga_physics_test_tube (Utilities|Casting|CastToBGA_PhysicsTestTube _array_element))
        (FortPhysicsSimulationLibrary|AddPhysicsImpulseatLocation _asbga_physics_test_tube (* (* (Transformation|GetForwardVector (Variables|Default|GetArrow)) "0.000000,0.000000,900.000000") (Variables|Default|GetLaunchMagnitudeScalar)) (Building|GetCentroid _asbga_physics_test_tube) true "None" true 0 "(MaxSimSizeAllowed=Large,bRejectEmbeddedInLandscape=True)")
        (else
          (bind _asbga_physics_test_tube_2 (Utilities|Casting|CastToBGA_PhysicsTestTube _array_element))
          (FortPhysicsSimulationLibrary|AddPhysicsImpulseatLocation _asbga_physics_test_tube_2 (* (* (Transformation|GetForwardVector (Variables|Default|GetArrow)) (Variables|Default|GetLaunchVelocity)) (Variables|Default|GetLaunchMagnitudeScalar)) (Transformation|GetWorldLocation (Variables|Default|GetArrow)) true "None" true 0 "(MaxSimSizeAllowed=Large,bRejectEmbeddedInLandscape=True)"))))
    (else
      (bind _outactors_2 (Collision|BoxOverlapActors (.location (Transformation|GetWorldTransform (Variables|Default|GetBox))) (* (Variables|/Script/Engine.BoxComponent|GetBoxExtent (Variables|Default|GetBox)) "1.200000,1.200000,1.200000") (Utilities|Array|MakeArray "ObjectTypeQuery3") "/Script/FortniteGame.FortPlayerPawnAthena" :!class /Script/Engine.KismetSystemLibrary))
      (if (> (Utilities|Array|Length _outactors_2) 0)
        (if (Variables|Default|GetisVerticalLauncher)
          (for _array_element_2 _outactors_2
            (bind _asfort_player_pawn_athena (Utilities|Casting|CastToFortPlayerPawnAthena _array_element_2))
            (Character|LaunchCharacter _asfort_player_pawn_athena "0.000000,0.000000,2200.000000" false true))
          (else
            (for _array_element_3 _outactors_2
              (bind _asfort_player_pawn_athena_2 (Utilities|Casting|CastToFortPlayerPawnAthena _array_element_3))
              (Character|LaunchCharacter _asfort_player_pawn_athena_2 (* (* (Transformation|GetForwardVector (Variables|Default|GetArrow)) (Variables|Default|GetLaunchVelocity)) (Variables|Default|GetLaunchMagnitudeScalar)) true true))))))))

(event Custom|ClientFX
  ; Timed Activation
  (Networking|SwitchHasAuthority
    (:Remote
      (bind _returnvalue (Niagara|SpawnSystematLocation "/Game/Effects/Fort_Effects/Effects/Environment/GeoThermalPlant/Systems/NSLW_WaterfallFX_Geothermal_Bottom.NSLW_WaterfallFX_Geothermal_Bottom" (Transformation|GetWorldLocation (Variables|Default|GetNiagara)) (Transformation|GetWorldRotation (Variables|Default|GetNiagara)) "0.600000,0.250000,0.250000" false))
      (Utilities|FlowControl|Delay 1.5)
      (Components|Activation|Deactivate _returnvalue)
      (Components|DestroyComponent _returnvalue))))

(fn ConstructionScript)
