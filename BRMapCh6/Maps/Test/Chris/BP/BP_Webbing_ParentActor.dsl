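; Test parent Blueprint for webbing static mesh actors, with no gameplay logic yet.
; The construction script reads the Webbing_1 socket on the static mesh component and draws a thick, temporary orange debug box at that location.
; The socket transform is taken in actor space but fed straight in as the box center, so the actor's world placement is not applied.
; BeginPlay, overlap and Tick events are empty stubs.
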(event EventBeginPlay)

(event Collision|EventActorBeginOverlap
  (params (OtherActor :type Actor%20Object%20Reference)))

(event EventTick
  (params (DeltaSeconds :type Float%20%28single-precision%29)))

(fn ConstructionScript
  (Rendering|Debug|DrawDebugBox :Center (.location (Transformation|GetSocketTransform (Variables|StaticMeshActor|GetStaticMeshComponent) "Webbing_1" "RTS_Actor")) :Extent "65.000000,65.000000,65.000000" :LineColor "(R=1.000000,G=0.778702,B=0.000000,A=1.000000)" :Duration 25.0 :Thickness 1e+01 :DepthPriority "World"))
