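; Test low-gravity volume Blueprint that applies the GE_LowGravity_UsingJumpStrong gameplay effect to Fortnite players inside its box.
; The construction script sizes the box from an extents variable.
; Begin overlap applies the effect to the pawn's player controller, and a pass five seconds after BeginPlay does the same for pawns already inside.
; End overlap only casts pawn and controller, so nothing here removes the effect on exit.
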(event EventBeginPlay
  (Utilities|FlowControl|Delay 5.0)
  (CallFunction|GetOverlappingPawns))

(event Custom|GetOverlappingPawns
  (for _array_element (Collision|GetOverlappingActors (Variables|Default|GetVolume) "/Script/FortniteGame.FortPlayerPawnAthena" :!class /Script/Engine.PrimitiveComponent)
    (bind _asfort_player_pawn_athena (Utilities|Casting|CastToFortPlayerPawnAthena _array_element))
    (bind _asfort_player_controller_athena (Utilities|Casting|CastToFortPlayerControllerAthena (Pawn|GetController _asfort_player_pawn_athena)))
    (Fortnite|ApplyGameplayEffectsToActor _asfort_player_controller_athena (Utilities|Array|MakeArray "/BRMapCh6/Maps/Test/Chris/BP/GE_LowGravity_UsingJumpStrong.GE_LowGravity_UsingJumpStrong_C") (Transformation|GetActorLocation _asfort_player_controller_athena))))

(event OnComponentBeginOverlap(Volume)
  (params (OverlappedComponent :type Primitive%20Component%20Object%20Reference) (OtherActor :type Actor%20Object%20Reference) (OtherComp :type Primitive%20Component%20Object%20Reference) (OtherBodyIndex :type Integer) (bFromSweep :type Boolean) (SweepResult :type Hit%20Result%20Structure%20%28by%20ref%29))
  (bind _asfort_player_pawn_athena (Utilities|Casting|CastToFortPlayerPawnAthena OtherActor))
  (bind _asfort_player_controller_athena (Utilities|Casting|CastToFortPlayerControllerAthena (Pawn|GetController _asfort_player_pawn_athena)))
  (Fortnite|ApplyGameplayEffectsToActor _asfort_player_controller_athena (Utilities|Array|MakeArray "/BRMapCh6/Maps/Test/Chris/BP/GE_LowGravity_UsingJumpStrong.GE_LowGravity_UsingJumpStrong_C") (Transformation|GetActorLocation _asfort_player_pawn_athena)))

(event OnComponentEndOverlap(Volume)
  (params (OverlappedComponent :type Primitive%20Component%20Object%20Reference) (OtherActor :type Actor%20Object%20Reference) (OtherComp :type Primitive%20Component%20Object%20Reference) (OtherBodyIndex :type Integer))
  (bind _asfort_player_pawn_athena (Utilities|Casting|CastToFortPlayerPawnAthena OtherActor))
  (Utilities|Casting|CastToFortPlayerControllerAthena (Pawn|GetController _asfort_player_pawn_athena)))

(fn ConstructionScript
  (Components|Box|SetBoxExtent (Variables|Default|GetVolume) (Variables|Default|GetBoxVolumeExtents) true))
