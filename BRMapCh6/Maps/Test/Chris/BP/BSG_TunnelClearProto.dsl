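; Prototype tunnel-clear actor that repeatedly sends a BP_TunnelPulser along a spline, starting after a reset delay.
; On authority it calls PrePulseFx, waits, spawns the pulser at the spline start, then a timeline moves it by distance along the spline.
; When the timeline finishes, it destroys the pulser and re-runs InitPulser after the reset delay.
; PrePulseFx's remote-only branch plays a warning alarm near the spline start, then fades it out.
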(event Custom|InitPulser
  (Networking|SwitchHasAuthority
    (:Authority
      (CallFunction|PrePulseFx)
      (Utilities|FlowControl|Delay 3.0)
      (bind _returnvalue (Game|SpawnActorfromClass "/BRMapCh6/Maps/Test/Chris/BP/BP_TunnelPulser.BP_TunnelPulser_C" (Math|Transform|MakeTransform (.location (Spline|GetTransformatDistanceAlongSpline (Variables|Default|GetSpline) 0.0 "World" false)) (.rotation (Spline|GetTransformatDistanceAlongSpline (Variables|Default|GetSpline) 0.0 "World" false))) "AlwaysSpawn" "MultiplyWithRoot" 0 0))
      (Variables|Default|SetPulserRef _returnvalue)
      (Utilities|FlowControl|Delay 0.1)
      (bind (_direction _newtrack_0) (Timeline|PulseTimeline PlayFromStart)
        (:Update
          (Transformation|SetActorTransform (Variables|Default|GetPulserRef) (Math|Transform|MakeTransform (.location (Spline|GetTransformatDistanceAlongSpline (Variables|Default|GetSpline) (* _newtrack_0 (Spline|GetSplineLength (Variables|Default|GetSpline))) "World" false)) (.rotation (Spline|GetTransformatDistanceAlongSpline (Variables|Default|GetSpline) (* _newtrack_0 (Spline|GetSplineLength (Variables|Default|GetSpline))) "World" false))) false true))
        (:Finished
          (Actor|DestroyActor (Variables|Default|GetPulserRef))
          (Utilities|FlowControl|Delay (Variables|Default|GetResetDelay))
          (CallFunction|InitPulser))))))

(event EventBeginPlay
  (Utilities|FlowControl|Delay (Variables|Default|GetResetDelay))
  (Networking|SwitchHasAuthority
    (:Authority
      (CallFunction|PrePulseFx)
      (Utilities|FlowControl|Delay 3.0)
      (bind _returnvalue (Game|SpawnActorfromClass "/BRMapCh6/Maps/Test/Chris/BP/BP_TunnelPulser.BP_TunnelPulser_C" (Math|Transform|MakeTransform (.location (Spline|GetTransformatDistanceAlongSpline (Variables|Default|GetSpline) 0.0 "World" false)) (.rotation (Spline|GetTransformatDistanceAlongSpline (Variables|Default|GetSpline) 0.0 "World" false))) "AlwaysSpawn" "MultiplyWithRoot" 0 0))
      (Variables|Default|SetPulserRef _returnvalue)
      (Utilities|FlowControl|Delay 0.1)
      (Timeline|PulseTimeline PlayFromStart))))

(event Custom|PrePulseFx
  (Networking|SwitchHasAuthority
    (:Remote
      (bind _returnvalue (Audio|SpawnSoundatLocation :Sound "/FNGameplayCues/Athena/Sounds/Gadgets/SecurityTurret/Security_Camera_Discreet_Alarm_Loop_Cue.Security_Camera_Discreet_Alarm_Loop_Cue" :Location (.location (Spline|GetTransformatDistanceAlongSpline (Variables|Default|GetSpline) (* 0.1 (Spline|GetSplineLength (Variables|Default|GetSpline))) "World" false)) :VolumeMultiplier 1.0 :PitchMultiplier 1.0 :StartTime 0.0 :AttenuationSettings "/Game/Sounds/Helios/POI/CastleForge/Spots/Doom_Forge_Thunder_Far_Distant_Focus_Att.Doom_Forge_Thunder_Far_Distant_Focus_Att" :bAutoDestroy true))
      (Utilities|FlowControl|Delay 3.0)
      (Audio|Components|Audio|FadeOut _returnvalue 1.0 0.0 "Linear" :!class /Script/Engine.AudioComponent))))

(fn ConstructionScript)
