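# M_Button_Proto is a simple Default Lit prototype material for a button, kept in a BRMapCh6 test map folder.
# A Color vector parameter is multiplied by a Glow scalar parameter, so material instances can adjust tint and intensity, alongside a fixed mid-value constant.
# Spline mesh usage is enabled, and automatic usage flagging in the editor is turned off.
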Material /BRMapCh6/Maps/Test/Chris/Materials/M_Button_Proto
graph {
 properties {bAutomaticallySetUsageInEditor=false, bUsedWithSplineMeshes=true, usedShadingModels=Default Lit}
 nodes [
  Constant() {r=0.5}
  Multiply(
   A=VectorParameter() {parameterName=Color, defaultValue=<rgba 0.2344 0.2344 0.2344 0>}.RGB
   B=ScalarParameter() {defaultValue=0.25, parameterName=Glow}
  )
 ]
}
