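# M_EmotionsMaterialParent is a Default Lit material in a test map folder, named as a parent for emotion materials.
# Its only node is a 2D texture parameter, EmotionTexture, which defaults to a Sleeping texture and lets instances swap the image.
# The graph shows no connection from that sampler to any material output.
# Automatic usage flag setting in the editor is turned off.
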Material /BRMapCh6/Maps/Test/Chris/Materials/M_EmotionsMaterialParent
graph {
 properties {bAutomaticallySetUsageInEditor=false, usedShadingModels=Default Lit}
 nodes [TextureSampleParameter2D() {parameterName=EmotionTexture, texture=&/Sleeping}]
}
