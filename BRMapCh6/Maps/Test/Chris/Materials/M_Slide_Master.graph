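# M_Slide_Master is a Default Lit master material in a test map folder, apparently intended for a slide.
# It holds a single vector parameter named Color with an orange-red default, for tinting in material instances.
# The file shows no connection from that parameter to any material output, and automatic usage setting in the editor is turned off.
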Material /BRMapCh6/Maps/Test/Chris/Materials/M_Slide_Master
graph {
 properties {bAutomaticallySetUsageInEditor=false, usedShadingModels=Default Lit}
 nodes [VectorParameter() {parameterName=Color, defaultValue=<rgba 0.3385 0.0168 0 1>}]
}
