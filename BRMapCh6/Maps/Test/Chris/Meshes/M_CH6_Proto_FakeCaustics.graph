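# Prototype translucent unlit material that fakes animated water caustics, stored in a Chapter 6 test map folder.
# It multiplies two samples of a foam swirl texture panned along perpendicular axes, then scales them by a brightness parameter and the alpha of the EmisiveColor parameter.
# A separate FOV-compensated distance fade blends between two scalar levels, likely for opacity, though output connections are not shown.
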Material /BRMapCh6/Maps/Test/Chris/Meshes/M_CH6_Proto_FakeCaustics
graph {
 properties {bAutomaticallySetUsageInEditor=false, blendMode=BLEND_Translucent, shadingModel=MSM_Unlit, usedShadingModels=Unlit}
 nodes [
  LinearInterpolate(
   A=ScalarParameter() {defaultValue=0.85}
   Alpha=MaterialFunctionCall!(
    Fade Offset=ScalarParameter() {defaultValue=1000, parameterName=FadeOffset}
    Fade Distance=ScalarParameter() {defaultValue=5000, parameterName=FadeDistance}
   ) {materialFunction=/Game/Global/Materials/Character_Parent_2023/Anime_Parent_2023/MaterialFunctions/MFC_DistanceFade_FOV_compensation}.Result
  ) {constB=0.15}
  n1 = TextureSample(UVs=Panner() {speedX=0.025}) {samplerType=SAMPLERTYPE_LinearColor, texture=/Game/Global/Textures/Default/Caustics/03_WaterFoam_AccentSwirls_H}
  n2 = TextureSample(UVs=Panner() {speedY=0.025}) {samplerType=SAMPLERTYPE_LinearColor, texture=&03_WaterFoam_AccentSwirls_H}
  Multiply(
   A=Multiply(A=VectorParameter() {parameterName=EmisiveColor, defaultValue=<rgba 0.0075 0.0399 0.0938 1>}.A)
   B=Multiply(A=Multiply(A=@n1.RGBA, B=@n2.RGBA), B=ScalarParameter() {defaultValue=0.25, parameterName=CausticsBrightness})
  )
 ]
}
