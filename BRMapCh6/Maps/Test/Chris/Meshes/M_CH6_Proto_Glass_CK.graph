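# Prototype glass material in a Chapter 6 test map folder, set up as translucent and unlit.
# A dark blue EmisiveColor vector parameter feeds a Multiply node, alongside a separate scalar parameter left at its default name.
# The graph has no textures or Fresnel, and its links to the material outputs are not shown.
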Material /BRMapCh6/Maps/Test/Chris/Meshes/M_CH6_Proto_Glass_CK
graph {
 properties {bAutomaticallySetUsageInEditor=false, blendMode=BLEND_Translucent, shadingModel=MSM_Unlit, usedShadingModels=Unlit}
 nodes [
  Multiply(A=VectorParameter() {parameterName=EmisiveColor, defaultValue=<rgba 0.0075 0.0399 0.0938 1>}.RGB)
  ScalarParameter() {defaultValue=0.9}
 ]
}
