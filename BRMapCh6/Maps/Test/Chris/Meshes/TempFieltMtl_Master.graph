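# Simple temporary master material in a Chapter 6 map test folder, using the Default Lit shading model.
# It exposes a Color parameter defaulting to dark olive green and a Roughness parameter defaulting to a fairly matte value, with no textures.
# Spline mesh usage is explicitly enabled, and automatic usage flagging in the editor is turned off.
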Material /BRMapCh6/Maps/Test/Chris/Meshes/TempFieltMtl_Master
graph {
 properties {bAutomaticallySetUsageInEditor=false, bUsedWithSplineMeshes=true, usedShadingModels=Default Lit}
 nodes [
  ScalarParameter() {defaultValue=0.85, parameterName=Roughness}
  VectorParameter() {parameterName=Color, defaultValue=<rgba 0.1129 0.1302 0.0508 1>}
 ]
}
